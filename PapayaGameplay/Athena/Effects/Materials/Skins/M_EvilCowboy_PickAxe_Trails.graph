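# Unlit, two-sided translucent mesh-particle material for the Evil Cowboy pickaxe trails, drawing stylized fire, green by default.
# World position offset combines SplineThicken with pulsing widths, distortion-texture waves and gradient-weighted local offsets.
# Panning noise, offset by a dynamic parameter, drives the two-color gradient and a normal-distorted flame shape cut by sine and sphere masks.
# Floor hardens opacity, modulated by particle alpha, near-camera fade and depth fade, which mobile skips.
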Material /PapayaGameplay/Athena/Effects/Materials/Skins/M_EvilCowboy_PickAxe_Trails
graph {
 properties {
  bUsedWithMeshParticles=true
  bUsedWithNiagaraMeshParticles=true
  blendMode=BLEND_Translucent
  pixelDepthOffsetMode=PDOM_Legacy
  shadingModel=MSM_Unlit
  twoSided=true
  usedShadingModels=Unlit
 }
 nodes [
  n1 = MaterialFunctionCall(Speed=ScalarParameter() {defaultValue=2, parameterName="Speed (S)"}) {materialFunction=&DebugTimeSine}
  n2 = LinearInterpolate(
   A=ScalarParameter() {defaultValue=5, parameterName=WidthBase_Min}
   B=ScalarParameter() {defaultValue=15, parameterName=WidthBase_Max}
   Alpha=Multiply(A=LinearInterpolate(Alpha=@n1.Result) {constA=0.5}, B=VertexColor().R)
  )
  n3 = Multiply(
   A=MaterialFunctionCall(Speed=ScalarParameter() {defaultValue=0.5, parameterName="Speed (S)_1"}) {materialFunction=&DebugTimeSine}.Result
   B=VertexColor().B
  )
  n4 = VertexColor()
  n5 = LinearInterpolate(
   A=ScalarParameter() {defaultValue=1, parameterName=WidthTip_Min}
   B=ScalarParameter() {defaultValue=8, parameterName=WidthTip_max}
   Alpha=LinearInterpolate(Alpha=Multiply(A=@n3, B=@n4.G)) {constA=0.5}
  )
  n6 = Multiply(A=Time(), B=ScalarParameter() {defaultValue=1, parameterName=Speed})
  n7 = Panner(Coordinate=TextureCoordinate() {coordinateIndex=1, uTiling=5}, Time=@n6) {speedY=0.8}
  n8 = TextureSample(UVs=@n7) {samplerType=SAMPLERTYPE_Normal, texture=&T_Weap_Distort_N_02}
  n9 = MaterialFunctionCall(Float2=TextureCoordinate()) {materialFunction=&BreakOutFloat2Components}
  n10 = LinearInterpolate(
   B=AppendVector(
    A=ComponentMask(None=@n8.RGB) {g=true, r=true}
    B=TextureSample(UVs=@n7) {texture=&T_Smooth_tile}.R
   )
   Alpha=OneMinus(None=Multiply(A=@n9.G))
  )
  n11 = Multiply(A=ScalarParameter() {defaultValue=10, parameterName=WaveIntensity}, B=VertexColor().B)
  n12 = Transform(None=VectorParameter() {parameterName=LocalPostioning}.RGB) {transformSourceType=TRANSFORMSOURCE_Local}
  n13 = TextureCoordinate()
  n14 = Power(
   Base=OneMinus(None=ComponentMask(None=@n13) {g=true})
   Exp=ScalarParameter() {defaultValue=2, parameterName=Edge Gradient Movement Control Exp}
  )
  n15 = Transform(None=VectorParameter() {parameterName=Initial WPO Offset, defaultValue=<rgba 0 -15 25 0>}.RGB) {transformSourceType=TRANSFORMSOURCE_Local}
  n16 = TextureCoordinate()
  n17 = Power(
   Base=OneMinus(None=ComponentMask(None=@n16) {g=true})
   Exp=ScalarParameter() {defaultValue=2, parameterName=Edge Gradient Movement Control Exp}
  )
  Add(
   A=Add(
    A=MaterialFunctionCall(WidthBase=@n2, WidthTip=@n5) {materialFunction=&SplineThicken}.WorldPositionOffset
    B=Multiply(A=@n10, B=@n11) {constB=10}
   )
   B=Add(A=Multiply(A=@n12, B=@n14), B=Multiply(A=@n15, B=@n17))
  )
  n18 = ScalarParameter() {defaultValue=1, group=Color, parameterName=Fire_Color_UScale}
  n19 = ScalarParameter() {defaultValue=1, group=Color, parameterName=Fire_Color_VScale, sortPriority=1}
  n20 = Panner(Coordinate=Multiply(A=TextureCoordinate() {coordinateIndex=1, uTiling=0.3, vTiling=0.7}, B=AppendVector(A=@n18, B=@n19)), Time=@n6) {speedY=-1.2}
  n21 = DynamicParameter() {paramNames [OffsetTExture, Param2, Param3, Param4]}
  n22 = LinearInterpolate(
   A=VectorParameter() {parameterName=ColorLow, defaultValue=<rgba 0 0.5052 0.0569 1>}.RGB
   B=VectorParameter() {parameterName=ColorHigh, defaultValue=<rgba 0.5987 1 0.038 1>}.RGB
   Alpha=TextureSample(UVs=Add(A=@n20, B=@n21.OffsetTExture)) {texture=/Game/Global/Textures/Default/Noise/T_FireNoise_02}.R
  )
  n23 = ScalarParameter() {parameterName=RandomSelection}
  n24 = Constant() {r=22}
  n25 = Ceil(None=Multiply(A=Clamp(None=@n23) {minDefault=0}, B=@n24))
  n26 = Ceil(None=Multiply(A=@n24, B=VertexColor().R))
  n27 = Constant()
  n28 = Constant() {r=1}
  n29 = DynamicParameter() {paramNames [OffsetTExture, Param2, Param3, Param4]}
  n30 = Multiply(
   A=TextureCoordinate() {coordinateIndex=1, vTiling=1.5}
   B=AppendVector(
    A=ScalarParameter() {defaultValue=1, group=Fire Shape, parameterName=Distortion_UScale, sortPriority=4}
    B=ScalarParameter() {defaultValue=1, group=Fire Shape, parameterName=Distortion_VScale, sortPriority=5}
   )
  )
  n31 = TextureSample(UVs=Panner(Coordinate=Add(A=@n29.OffsetTExture, B=@n30), Time=@n6) {speedY=-2}) {samplerType=SAMPLERTYPE_Normal, texture=&T_Weap_Distort_N_02}
  n32 = ScalarParameter() {defaultValue=1, group=Fire Shape, parameterName=Fire_MacroShape_UScale}
  n33 = ScalarParameter() {defaultValue=1, group=Fire Shape, parameterName=Fire_MacroShape_VScale, sortPriority=1}
  n34 = MaterialFunctionCall(Float2=Multiply(A=TextureCoordinate(), B=AppendVector(A=@n32, B=@n33))) {materialFunction=&BreakOutFloat2Components}
  n35 = Multiply(A=@n34.G)
  n36 = Divide(
   A=LinearInterpolate(A=ComponentMask(None=@n31.RGB) {g=true, r=true}, Alpha=@n35) {constB=0}
   B=ScalarParameter() {defaultValue=3, parameterName=UV_Distortion}
  ) {constB=20}
  n37 = TextureCoordinate() {coordinateIndex=1, uTiling=0.8, vTiling=0.9}
  n38 = AppendVector(
   A=ScalarParameter() {defaultValue=1, group=Fire Shape, parameterName=Fire_Micro_UScale, sortPriority=3}
   B=ScalarParameter() {defaultValue=1, group=Fire Shape, parameterName=Fire_Micro_VScale, sortPriority=3}
  )
  n39 = Panner(
   Coordinate=MaterialFunctionCall(XY=Multiply(A=@n37, B=@n38)) {materialFunction=&Swizzle}.YX
   Time=@n6
  ) {speedX=-2, speedY=0.2}
  n40 = TextureSample(UVs=Add(A=@n21.OffsetTExture, B=Add(A=@n36, B=@n39))) {samplerType=SAMPLERTYPE_LinearColor, texture=&T_NoisePack_LinearColor}
  n41 = Saturate(None=Multiply(A=If(A=@n25, B=@n26, "A > B"=@n27, "A == B"=@n28, "A < B"=@n27), B=@n40.R))
  n42 = Multiply(A=@n22, B=ScalarParameter() {defaultValue=2, parameterName=EmissiveAmount}) {constA=5}
  MaterialFunctionCall(Emissive=MaterialFunctionCall(
    LDR Input=@n22
    HDR Dynamic Range=ScalarParameter() {defaultValue=5, parameterName=HDR Dynamic Range}
    HDR Tint=@n42
    HDR Intensity Multiplier=LinearInterpolate(Alpha=@n41) {constA=1, constB=5}
   ) {materialFunction=&DeriveHDRfromLDR}.RemappedRange) {materialFunction=&MF_FX_EmissiveAdjustments}
  Multiply(A=@n42, B=ParticleColor().RGB)
  n43 = Multiply(A=@n34.R) {constB=0.5}
  n44 = Multiply(A=Power(Base=Sine(None=@n43), Exp=LinearInterpolate(Alpha=@n35) {constA=3, constB=2}) {constExponent=4}, B=@n35)
  n45 = TextureCoordinate() {uTiling=1.25}
  n46 = Constant2Vector() {g=0.05, r=0.6}
  n47 = Saturate(None=Multiply(A=SphereMask(A=@n45, B=@n46) {attenuationRadius=0.95, hardnessPercent=0}) {constB=15})
  n48 = Multiply(A=SphereMask(
    A=TextureCoordinate() {uTiling=0, vTiling=1.1}
    B=Constant2Vector() {g=0.05, r=0.6}
    Hardness=ScalarParameter() {parameterName=Bottom_Edge_Hardness}
   ) {attenuationRadius=1, hardnessPercent=0}) {constB=15}
  n49 = Multiply(A=@n40.R, B=Multiply(A=Multiply(A=@n44, B=@n47), B=Saturate(None=@n48)))
  n50 = LinearInterpolate(Alpha=@n35) {constA=2}
  n51 = Sine(None=Multiply(A=@n34.R) {constB=0.45})
  n52 = Add(
   A=Power(Base=@n35) {constExponent=1}
   B=Power(Base=OneMinus(None=@n51), Exp=LinearInterpolate(Alpha=@n31.R) {constA=0.5, constB=1.5}) {constExponent=0.5}
  )
  n53 = Multiply(A=Multiply(A=Power(Base=@n49, Exp=@n50), B=Round(None=@n52)), B=ScalarParameter() {defaultValue=25, parameterName=OpacityMaskMultiplier}) {constB=3}
  n54 = DepthFade(Opacity=Floor(None=@n53), FadeDistance=ScalarParameter() {defaultValue=55, parameterName=DepthFadeDistance})
  n55 = ScalarParameter() {defaultValue=100, parameterName=DepthFadeDistance_Outside_Mesh}
  Multiply(A=@n54, B=OneMinus(None=DepthFade(FadeDistance=@n55)))
  n56 = MaterialFunctionCall(Distance=ScalarParameter() {defaultValue=128, parameterName=NearCamFade_Distance}) {materialFunction=&Near_Camera_Fade}
  n57 = ScalarParameter() {defaultValue=4, parameterName=NearCamFade_Power}
  Multiply(
   A=Multiply(A=ShadingPathSwitch(Default=@n54, Mobile=@n53) {desc=Migrated from FeatureLevelSwitch}, B=ParticleColor().A)
   B=Clamp(None=Power(Base=@n56.Result, Exp=@n57)) {minDefault=0.1}
  )
  VertexColor()
 ]
}
